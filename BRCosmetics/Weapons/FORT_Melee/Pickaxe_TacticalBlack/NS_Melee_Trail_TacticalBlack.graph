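# Niagara system for the swing trail of the Tactical Black pickaxe cosmetic, drawn as a short-lived ribbon.
# A single FN_AnimTrail emitter uses shared anim-trail modules to spawn, initialize bone indices and update particles, with a socket reader dependency and a user RibbonIDOffset parameter.
# Over particle life, color shifts from muted blue-grey through red to dark red, tinted by System.ToDColor, while an alpha curve fades the trail out.
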NiagaraSystem /BRCosmetics/Weapons/FORT_Melee/Pickaxe_TacticalBlack/NS_Melee_Trail_TacticalBlack
dependencies {
 usedDataInterfaces [&NiagaraDataInterfaceColorCurve, &NiagaraDataInterfaceCurve, &NiagaraDataInterfaceSocketReader]
 usedDynamicInputs []
 usedModules [
  &self:SystemSpawnScript.SystemScriptSource.SystemScriptGraph.NiagaraNodeAssignment_1.SetVariables_#1
  &SystemState
  &self:SystemSpawnScript.SystemScriptSource.SystemScriptGraph.NiagaraNodeAssignment_2.SetVariables_#2
 ]
}
emitters {FN_AnimTrail {
  emitterSpawnScript [&FN_RibbonIDOffsetUtility()]
  emitterUpdateScript [
   &EmitterState()
   &FN_SpawnAnimTrail()
  ]
  particleSpawnScript [
   &FN_InitializeAnimTrailBoneIndecies()
   &FN_InitializeAnimTrail(Color=<rgba 1 1 1 1.25>, Lifetime=0.22)
  ]
  particleUpdateScript [
   &SolveForcesAndVelocity(Clamp Velocity=false, Limit Acceleration=false)
   &FN_UpdateAnimTrail() {module=FN_UpdateAnimTrail001}
   &ScaleColor(
    Scale Alpha=&FloatFromCurve(
     FloatCurve=DataInterface() {propertyValues {Curve {keys [{arriveTangent=-4.9584, interpMode=RCIM_Cubic, leaveTangent=-4.9584, tangentMode=RCTM_User, time=0.3, value=1}, <at 0.7 0 RCIM_Cubic>]}}}
     Scale Curve=0.75
    )
    Scale RGB=&MakeVectorFromLinearColorRGB(Color=&MultiplyLinearColor(
      Linear Color=&ColorFromCurve(ColorCurve=DataInterface() {propertyValues {
         AlphaCurve {keys []}
         BlueCurve {keys [
           {arriveTangent=0.0851, interpMode=RCIM_Cubic, time=0, value=0.25}
           {arriveTangent=-0.2496, interpMode=RCIM_Cubic, leaveTangent=-0.2496, time=0.115, value=0.0698}
           <at 0.9224 0.0197>
          ]}
         GreenCurve {keys [
           {arriveTangent=0.0511, interpMode=RCIM_Cubic, time=0, value=0.2175}
           {arriveTangent=-0.217, interpMode=RCIM_Cubic, leaveTangent=-0.217, time=0.115, value=0.0637}
           <at 0.9224 0.0174>
          ]}
         RedCurve {keys [
           {arriveTangent=-0.4334, interpMode=RCIM_Cubic, time=0, value=0.2}
           {arriveTangent=-0.1224, interpMode=RCIM_Cubic, leaveTangent=-0.1224, time=0.115, value=0.425}
           <at 0.9224 0.0871>
          ]}
        }})
      Scale Factor=&MakeVector4_FromLinearColor(LinearColor=$System.ToDColor)
     ))
    ScaleA=true
    ScaleRGB=true
   )
  ]
  rendererClasses [&NiagaraRibbonRendererProperties]
  renderers [{rendererClass=&NiagaraRibbonRendererProperties}]
 }}
summary {
 systemName=NS_Melee_Trail_TacticalBlack
 emitters [{emitterName=FN_AnimTrail, rendererClasses [&NiagaraRibbonRendererProperties]}]
 userVariables [User.RibbonIDOffset {type=NiagaraInt32}]
}
